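# Material function that plays a staged transition effect on a gun prop, from masked-out through glowing purple stylized stages to its input material attributes.
# MaterialCurve A–E scalars drive SphereMask wipes along Z for opacity, stage blends, grid fill and darkening.
# Procedural grids use Fmod and Round on offset pre-skinned position, weighted by pre-skinned normals, with MaterialCurveF varying cell size.
# Normal-derived curvature and a mask texture feed the emissive detail.
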MaterialFunction /BRCosmetics/Animation/Props/Gadgets_High_Motion/Materials/MF_HighMotion_Guntransition
graph {
 properties {previewBlendMode=BLEND_Masked}
 nodes [
  ComponentMask() {g=true, r=true}
  n1 = VertexInterpolator(VS=PreSkinnedNormal())
  ComponentMask(None=@n1.PS) {b=true}
  Constant()
  n2 = FunctionInput() {inputType=FunctionInput_MaterialAttributes}
  n3 = GetMaterialAttributes(None=@n2)
  n4 = Constant()
  n5 = BlendMaterialAttributes(A=NamedRerouteUsage(), B=NamedRerouteUsage(), Alpha=NamedRerouteUsage()) {vertexAttributeBlendType=UseA}
  n6 = NamedRerouteUsage()
  FunctionOutput(None=BlendMaterialAttributes(
    A=SetMaterialAttributes!(MaterialAttributes=@n3.MaterialAttributes, OpacityMask=@n4) {attributeSetTypes [#1]}
    B=BlendMaterialAttributes(A=@n5, B=@n3.MaterialAttributes, Alpha=@n6) {vertexAttributeBlendType=UseA}
    Alpha=NamedRerouteUsage()
   ) {vertexAttributeBlendType=UseA})
  n7 = Add(A=NamedRerouteUsage(), B=ScalarParameter() {defaultValue=0.9, parameterName=Boost})
  n8 = Saturate(None=ScalarParameter() {parameterName=MaterialCurveF})
  n9 = LinearInterpolate(Alpha=@n8) {constA=0.08, constB=0.025}
  n10 = Saturate(None=Power(Base=@n8))
  n11 = MaterialFunctionCall(Float3=@n7) {materialFunction=&BreakOutFloat3Components}
  n12 = LinearInterpolate(
   A=@n9
   B=Multiply(A=@n9, B=LinearInterpolate(Alpha=@n10) {constA=0.2}) {constB=0.1}
   Alpha=Saturate(None=Power(Base=@n11.B) {constExponent=0.5})
  ) {constB=0.05}
  n13 = Fmod(A=Multiply(A=@n7, B=ScalarParameter() {defaultValue=0.6, parameterName=asdfasdf}), B=@n12)
  n14 = ScalarParameter() {defaultValue=0.52, parameterName=NegativeSpace}
  n15 = Round(None=Multiply(A=Divide(A=@n13, B=@n12), B=@n14))
  n16 = ComponentMask(None=@n15) {b=true}
  n17 = NamedRerouteUsage()
  LinearInterpolate(A=@n16, B=Max(A=ComponentMask(None=@n15) {g=true, r=true}, B=@n16), Alpha=@n17)
  n18 = FunctionInput(Preview=VertexInterpolator(VS=PreSkinnedPosition()).PS) {inputName=Pre Skinned Position}
  n19 = ScalarParameter() {defaultValue=1.1, parameterName=FmodSimple}
  n20 = Fmod(A=Abs(None=@n18), B=@n19)
  n21 = Round(None=Multiply(A=Divide(A=@n20, B=@n19), B=@n14))
  Max(A=ComponentMask(None=@n21) {r=true}, B=ComponentMask(None=@n21) {g=true})
  n22 = Abs(None=Add(A=NamedRerouteUsage(), B=ScalarParameter() {defaultValue=-0.56, parameterName=CenterOffset}))
  n23 = Saturate(None=Multiply(A=ComponentMask(None=@n22) {b=true}) {constB=2})
  n24 = Constant()
  n25 = LinearInterpolate(Alpha=ScalarParameter() {defaultValue=1, parameterName=MaterialCurveA}) {constA=-0.1, constB=2}
  NamedRerouteDeclaration(None=Multiply(A=SphereMask(A=@n23, B=@n24, Radius=@n25) {attenuationRadius=1, hardnessPercent=50})) {name=InitialOpacityMask, nodeColor=<rgba 0.8047 0 1 1>}
  n26 = PreSkinnedLocalBounds()
  NamedRerouteDeclaration(None=Divide(A=@n18, B=Subtract(A=@n26.Max, B=@n26.Max))) {name=Object Gradient, nodeColor=<rgba 1 0 0.8906 1>}
  n27 = Abs(None=@n1.PS)
  n28 = ComponentMask(None=@n27) {b=true, g=true}
  n29 = ComponentMask(None=@n27) {b=true, r=true}
  NamedRerouteDeclaration(None=Multiply(A=@n16, B=Max(A=@n28, B=@n29))) {name=HorizontalGrid, nodeColor=<rgba 0.1719 1 0 1>}
  n30 = NamedRerouteUsage()
  n31 = ScalarParameter() {defaultValue=0.2, parameterName=MaterialCurveE}
  NamedRerouteDeclaration(None=SphereMask(A=ComponentMask(None=@n30) {b=true}, B=Constant(), Radius=LinearInterpolate(Alpha=@n31) {constB=2}) {hardnessPercent=66}) {name=Final Transition, nodeColor=<rgba 0 0.6016 1 1>}
  NamedRerouteDeclaration(None=@n2) {name=MaterialAttributes, nodeColor=<rgba 0.0781 0 1 1>}
  n32 = NamedRerouteUsage()
  n33 = Constant3Vector() {constant=<rgba 0.6562 0.1005 0.6384 1>}
  n34 = Constant3Vector() {constant=<rgba 9.4167 0 15 1>}
  n35 = NamedRerouteUsage()
  n36 = Constant3Vector() {constant=<rgba 0.0156 0.0067 0.0153 1>}
  n37 = Constant3Vector() {constant=<rgba 9.4167 0 15 1>}
  n38 = NamedRerouteUsage()
  n39 = Constant()
  n40 = SetMaterialAttributes!(
   MaterialAttributes=@n32
   OpacityMask=Constant() {r=1}
   EmissiveColor=LinearInterpolate(A=LinearInterpolate(A=@n33, B=@n34, Alpha=@n35), B=LinearInterpolate(A=@n36, B=@n37, Alpha=@n38), Alpha=NamedRerouteUsage())
   BaseColor=Constant()
   Specular=@n39
   Roughness=Constant() {r=1}
  ) {attributeSetTypes [
    #1
    #2
    #3
    #4
    #5
   ]}
  NamedRerouteDeclaration(None=@n40) {name=Stage1, nodeColor=<rgba 1 0 0.75 1>}
  n41 = Constant3Vector() {constant=<rgba 0.1633 0 0.6562 1>}
  n42 = NamedRerouteUsage()
  n43 = LinearInterpolate(
   B=TextureSample() {samplerType=SAMPLERTYPE_Masks, texture=/BRCosmetics/Animation/Props/Gadgets_High_Motion/Textures/T_HighMotionmat_lowpoly_Mask}.G
   Alpha=@n17
  )
  n44 = SetMaterialAttributes!(
   MaterialAttributes=NamedRerouteUsage()
   OpacityMask=Constant() {r=1}
   EmissiveColor=LinearInterpolate(A=@n41, B=Constant3Vector() {constant=<rgba 1.9909 0 8 1>}, Alpha=Max(A=@n42, B=@n43))
   BaseColor=@n41
   Specular=@n39
  ) {attributeSetTypes [
    #1
    #2
    #3
    #4
   ]}
  NamedRerouteDeclaration(None=@n44) {name=Stage2, nodeColor=<rgba 0 1 0.7422 1>}
  n45 = NamedRerouteUsage()
  n46 = ScalarParameter() {defaultValue=1, parameterName=MaterialCurveB}
  NamedRerouteDeclaration(None=SphereMask(A=ComponentMask(None=@n45) {b=true}, B=Constant(), Radius=LinearInterpolate(Alpha=@n46) {constB=1.5}) {hardnessPercent=66}) {name=Stage 2 Transition, nodeColor=<rgba 0 0.6016 1 1>}
  n47 = NamedRerouteUsage()
  n48 = ScalarParameter() {defaultValue=1, parameterName=MaterialCurveD}
  NamedRerouteDeclaration(None=SphereMask(A=ComponentMask(None=@n47) {b=true}, B=Constant(), Radius=LinearInterpolate(Alpha=@n48) {constB=2}) {hardnessPercent=66}) {name=GridFillMask, nodeColor=<rgba 0 0.6016 1 1>}
  n49 = GetMaterialAttributes(None=@n32) {attributeGetTypes [#6]}
  n50 = Abs(None=ComponentMask(None=@n49.Normal) {g=true})
  n51 = Abs(None=ComponentMask(None=@n49.Normal) {r=true})
  NamedRerouteDeclaration(None=Power(Base=Max(A=@n50, B=@n51)) {constExponent=4}) {name=Curvature, nodeColor=<rgba 1 0.0938 0 1>}
  n52 = NamedRerouteUsage()
  n53 = ScalarParameter() {parameterName=MaterialCurveC}
  NamedRerouteDeclaration(None=SphereMask(A=ComponentMask(None=@n52) {b=true}, B=Constant(), Radius=LinearInterpolate(Alpha=@n53) {constB=1.5}) {hardnessPercent=66}) {name=Stage 2 Darkness, nodeColor=<rgba 0 0.6016 1 1>}
  NamedRerouteUsage()
  NamedRerouteUsage()
 ]
}
